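# Material function for the MistRaven Cowl character skin that adds an animated Styx emissive transition and subsurface fake skin to incoming material attributes.
# Panning noise, normal-map distortion, detail and smoke layers use mesh or pre-skinned-position UVs, with emissive, body, face and neck masks shaping glow.
# IsNeck switches neck-specific branches, while IsHead and masks blend subsurface color and opacity.
# Skin tint varies by time of day, and emissive passes through MF_FX_EmissiveAdjustments.
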MaterialFunction /BRCosmetics/Characters/Player/Female/Medium/Bodies/F_MED_MistRaven_Peck/Skins/Cowl/Materials/MF_Misraven_Cowl_Transition
graph {nodes [
  Constant() {r=0.5}
  n1 = StaticSwitchParameter(True=NamedRerouteUsage(), False=NamedRerouteUsage()) {group=Styx_Transit, parameterName=IsNeck}
  n2 = ScalarParameter() {defaultValue=0.38, group=MistRaven, parameterName=Emissive_Global}
  n3 = ComponentMask(None=NamedRerouteUsage()) {b=true}
  n4 = LinearInterpolate(
   A=ComponentMask(None=NamedRerouteUsage()) {b=true}
   B=OneMinus(None=Power(Base=@n3) {constExponent=1.5})
   Alpha=ScalarParameter() {parameterName=IsHead}
  ) {constAlpha=0}
  n5 = SetMaterialAttributes!(
   MaterialAttributes=NamedRerouteUsage()
   EmissiveColor=MaterialFunctionCall(Emissive=Multiply(A=@n1, B=@n2) {constB=0.5}) {materialFunction=&MF_FX_EmissiveAdjustments}.Result
   SubsurfaceColor=NamedRerouteUsage()
   Opacity=LinearInterpolate(B=ScalarParameter() {defaultValue=0.25, group=MistRaven, parameterName=SSS_Opacity}, Alpha=@n4) {constA=1}
   ShadingModel=ShadingModel() {shadingModel=MSM_Subsurface}
  ) {attributeSetTypes [
    #1
    #2
    #3
    #4
   ]}
  FunctionOutput(None=@n5)
  n6 = ScalarParameter() {defaultValue=1, group=Styx_Transit, parameterName=EmissiveFadeMin}
  n7 = ScalarParameter() {defaultValue=1.5, group=Styx_Transit, parameterName=EmissiveFadeMax}
  n8 = NamedRerouteUsage()
  n9 = GetMaterialAttributes(None=FunctionInput() {inputName=Material, inputType=FunctionInput_MaterialAttributes}) {attributeGetTypes [#1]}
  n10 = ChannelMaskParameter(None=NamedRerouteDeclaration(None=@n9.EmissiveColor) {name=ORG_EM, nodeColor=<rgba 0 0.6016 1 1>}) {group=Styx_Transit, maskChannel=Green, parameterName=UseEmissiveAsMask, defaultValue=<rgba 0 1 0 0>}
  n11 = NamedRerouteDeclaration(None=FunctionInput() {inputName=EmissiveTextureSampler}) {name=EmissiveIN, nodeColor=<rgba 1 0.5625 0 1>}
  n12 = ChannelMaskParameter(None=@n11) {group=Styx_Transit, maskChannel=Green, parameterName=UseEmissiveAsMask, defaultValue=<rgba 0 1 0 0>}
  n13 = Saturate(None=StaticSwitchParameter(True=@n10.R, False=NamedRerouteDeclaration(None=@n12.R) {name=EmissiveMask, nodeColor=<rgba 1 0.1406 0 1>}) {group=Styx_Transit, parameterName=UseOrgEM})
  n14 = Multiply(A=OneMinus(None=@n13), B=@n13)
  NamedRerouteDeclaration(None=Multiply(A=LinearInterpolate(A=@n6, B=@n7, Alpha=@n8), B=OneMinus(None=@n14))) {name=EmissiveFadeMask, nodeColor=<rgba 1 0.5625 0 1>}
  n15 = Divide(
   A=StaticComponentMaskParameter(None=NamedRerouteUsage()) {defaultB=true, defaultR=true, parameterName=WorldUVChannel}
   B=ScalarParameter() {defaultValue=20, group=Styx_Transit, parameterName=WorldUVSize}
  )
  n16 = TextureCoordinate()
  n17 = ScalarParameter() {defaultValue=3, group=Styx, parameterName=StyxMainNoiseTilling}
  NamedRerouteDeclaration(None=StaticSwitchParameter(True=NamedRerouteDeclaration(None=@n15) {name=UVVis, nodeColor=<rgba 0.0312 1 0 1>}, False=Multiply(A=@n16, B=@n17)) {group=Styx_Transit, parameterName=UsePreskinnedAlignedUV}) {name=UV_withParallax_A, nodeColor=<rgba 1 0.6562 0 1>}
  n18 = NamedRerouteUsage()
  n19 = AppendVector(
   A=ScalarParameter() {group=Styx, parameterName=StyxNoise_Speed2X}
   B=ScalarParameter() {defaultValue=-0.1, group=Styx, parameterName=StyxNoise_Speed2Y}
  )
  n20 = TextureSample(UVs=Add(A=Panner(Coordinate=@n18, Speed=@n19), B=NamedRerouteUsage()), Tex=NamedRerouteUsage()) {samplerSource=SSM_Wrap_WorldGroupSettings}
  n21 = ScalarParameter() {defaultValue=5, group=Styx, parameterName=StyxNoise_Contrast}
  n22 = LinearInterpolate(
   A=Power(Base=@n20.R, Exp=ScalarParameter() {defaultValue=5, group=Styx, parameterName=StyxNoise_Contrast2})
   B=@n20.R
   Alpha=Saturate(None=Power(Base=@n20.R, Exp=@n21))
  )
  n23 = ScalarParameter() {defaultValue=1, group=Styx, parameterName=MainNoiseIntensity}
  n24 = NamedRerouteUsage()
  NamedRerouteDeclaration(None=LinearInterpolate(A=Multiply(A=@n22, B=@n23), B=NamedRerouteUsage(), Alpha=ComponentMask(None=@n24) {g=true})) {name=MainNoisePattern, nodeColor=<rgba 0 1 0.3203 1>}
  n25 = Panner(
   Coordinate=Add(A=NamedRerouteUsage(), B=NamedRerouteUsage())
   Speed=AppendVector(
    A=ScalarParameter() {group=Styx_Detail, parameterName=Styx_Detail_Noise_SpeedX}
    B=ScalarParameter() {defaultValue=-0.1, group=Styx_Detail, parameterName=Styx_Detail_Noise_SpeedY}
   )
  )
  n26 = TextureSampleParameter2D(UVs=@n25) {
   group=Styx_Detail
   parameterName=StyxNoiseDetail
   samplerSource=SSM_Wrap_WorldGroupSettings
   samplerType=SAMPLERTYPE_Masks
   texture=/BRCosmetics/Characters/Player/Male/Medium/Bodies/M_MED_GateHound_Spectral/Textures/T_StyxGateHound_DetailNoise_M
  }
  n27 = LinearInterpolate(
   A=ScalarParameter() {group=Styx_Detail, parameterName=Styx_Detail_Noise_Low}
   B=ScalarParameter() {defaultValue=1, group=Styx_Detail, parameterName=Styx_Detail_Noise_High}
   Alpha=@n26.R
  )
  n28 = ComponentMask(None=NamedRerouteUsage()) {g=true}
  NamedRerouteDeclaration(None=LinearInterpolate(A=Saturate(None=@n27), B=NamedRerouteUsage(), Alpha=@n28)) {name=DetailNoise, nodeColor=<rgba 1 0.0938 0 1>}
  n29 = FunctionInput() {bUsePreviewValueAsDefault=true, inputName=PreSkinnedLocation, sortPriority=2, previewValue=<1 0 0 1>}
  NamedRerouteDeclaration(None=@n29) {name=PreSkinned Loc, nodeColor=<rgba 1 0 0.3984 1>}
  n30 = ComponentMask(None=NamedRerouteUsage()) {g=true}
  n31 = LinearInterpolate(
   A=VectorParameter() {group=MistRaven, parameterName=Leg_color_Emissive, defaultValue=<rgba 0.5217 0.0257 0.1466 1>}.RGB
   B=Multiply(A=VectorParameter() {group=MistRaven, parameterName=Leg_color_Emissive, defaultValue=<rgba 0.5217 0.0257 0.1466 1>}.RGB, B=Constant() {r=3})
   Alpha=Power(Base=@n30) {constExponent=8}
  )
  n32 = Power(Base=@n30) {constExponent=1}
  n33 = Constant() {r=0.446}
  NamedRerouteDeclaration(None=LinearInterpolate(A=NamedRerouteUsage(), B=Multiply(A=@n31, B=@n32), Alpha=Multiply(A=@n30, B=@n33))) {name=bottom of the legs, nodeColor=<rgba 1 0.1406 0 1>}
  n34 = Subtract(A=NamedRerouteUsage(), B=VectorParameter() {group=StyxMod, parameterName=NeckPOS, defaultValue=<rgba 0 -5 135 1>}.RGB)
  n35 = ComponentMask(None=@n34) {g=true}
  n36 = Saturate(None=Divide(A=Abs(None=@n35), B=ScalarParameter() {defaultValue=5, parameterName=StretchUnit}) {constB=3})
  NamedRerouteDeclaration(None=OneMinus(None=SmoothStep(Value=@n36) {constMin=0.3})) {name=StretchMask, nodeColor=<rgba 0.5469 1 0 1>}
  NamedRerouteDeclaration(None=SmoothStep(
    Max=ScalarParameter() {defaultValue=1, group=Styx_Transit, parameterName=AreaMax}
    Value=ChannelMaskParameter(None=@n11) {group=Styx_Transit, maskChannel=Blue, parameterName=AreaMaskChannel, defaultValue=<rgba 0 0 1 0>}.R
   )) {name=AreaMask, nodeColor=<rgba 1 0.7031 0 1>}
  n37 = NamedRerouteUsage()
  n38 = TextureSampleParameter2D() {
   group=MistRaven
   parameterName=BodyMask
   samplerSource=SSM_Wrap_WorldGroupSettings
   samplerType=SAMPLERTYPE_Masks
   texture=&T_BlackMask
  }
  n39 = MaterialFunctionCall(UVs=TextureCoordinate(), "Rotation Angle (0-1)"=LinearInterpolate(Alpha=@n38.R) {constB=0.25}) {materialFunction=&CustomRotator}
  n40 = Add(A=Multiply(A=@n37) {constB=0.0015}, B=@n39.Rotated Values)
  n41 = Constant2Vector() {g=16, r=4}
  n42 = TextureSampleParameter2D(UVs=Panner(Coordinate=Multiply(A=@n40, B=@n41) {constB=16}) {speedX=0.01, speedY=0.05}) {
   group=StyxMod
   parameterName=DistortSmoke
   samplerType=SAMPLERTYPE_Grayscale
   texture=/BRCosmetics/Characters/Player/Male/Medium/Bodies/M_MED_GateHound_Spectral/Textures/T_FractalDistortMask_GateHound
  }
  n43 = TextureObjectParameter() {
   group=Styx
   parameterName=Styx Mask
   samplerType=SAMPLERTYPE_Masks
   texture=/BRCosmetics/Characters/Player/Male/Medium/Bodies/M_MED_GateHound_Spectral/Textures/T_StyxGateHound_Mask
  }
  n44 = TextureSample(UVs=Panner(Coordinate=TextureCoordinate()) {speedX=0.05, speedY=0.05}, Tex=@n43) {samplerSource=SSM_Wrap_WorldGroupSettings}
  n45 = Multiply(A=Multiply(A=@n42.R, B=@n44.G))
  n46 = Add(A=Multiply(A=@n37) {constB=0.001}, B=@n39.Rotated Values)
  n47 = Constant2Vector() {g=16, r=3.5}
  n48 = TextureSampleParameter2D(UVs=Panner(Coordinate=Multiply(A=@n46, B=@n47) {constB=14}) {speedX=-0.005, speedY=-0.1}) {
   group=StyxMod
   parameterName=DistortSmoke
   samplerType=SAMPLERTYPE_Grayscale
   texture=&T_FractalDistortMask_GateHound
  }
  n49 = Multiply(A=Multiply(A=@n44.B, B=@n48.R)) {constB=1.3}
  n50 = LinearInterpolate(Alpha=Power(Base=Max(A=@n45, B=@n49)) {constExponent=1.7}) {constB=0.5}
  n51 = TextureSample() {
   samplerSource=SSM_Wrap_WorldGroupSettings
   samplerType=SAMPLERTYPE_Masks
   texture=/BRCosmetics/Characters/Player/Female/Medium/Bodies/F_MED_MistRaven_Peck/Skins/Cowl/Textures/T_MistRaven_Peck_Body_TA_mask4
  }
  n52 = LinearInterpolate(A=LinearInterpolate(Alpha=@n51.G) {constA=1, constB=0.5}, Alpha=Power(Base=@n51.G) {constExponent=3}) {constB=0.1}
  n53 = NamedRerouteUsage()
  n54 = Multiply(
   A=VectorParameter() {group=MistRaven, parameterName=Glowline_color, defaultValue=<rgba 0.1651 0.0044 0.0242 1>}.RGB
   B=LinearInterpolate(Alpha=Power(Base=@n53)) {constA=20, constB=0}
  ) {constB=15}
  n55 = TextureSampleParameter2D() {
   group=06_Emissive
   parameterName=Emissive Mask
   samplerSource=SSM_Wrap_WorldGroupSettings
   samplerType=SAMPLERTYPE_Masks
   texture=&T_BlackMask
  }
  n56 = LinearInterpolate(B=Multiply(A=Saturate(None=@n50), B=Multiply(A=@n52, B=@n54)), Alpha=@n55.B)
  n57 = Multiply(A=ComponentMask(None=NamedRerouteUsage()) {r=true}) {constB=2}
  n58 = Power(Base=OneMinus(None=Saturate(None=@n57)))
  NamedRerouteDeclaration(None=StaticSwitchParameter(True=Constant(), False=Multiply(A=@n56, B=@n58)) {group=Styx_Transit, parameterName=IsNeck}) {name=inside, nodeColor=<rgba 0.4062 0 1 1>}
  n59 = NamedRerouteUsage()
  n60 = Multiply(A=@n59) {constB=0.2}
  n61 = ScalarParameter() {parameterName=IsHead}
  NamedRerouteDeclaration(None=LinearInterpolate(A=LinearInterpolate(B=@n60, Alpha=@n61), B=@n59, Alpha=@n4)) {name=Subsurface Color, nodeColor=<rgba 0 0.1797 1 1>}
  n62 = Divide(
   A=ComponentMask(None=NamedRerouteUsage()) {b=true, g=true}
   B=ScalarParameter() {defaultValue=20, group=Styx_Transit, parameterName=WorldUVSize}
  )
  n63 = TextureCoordinate()
  n64 = ScalarParameter() {defaultValue=3, group=Styx, parameterName=StyxMainNoiseTilling}
  NamedRerouteDeclaration(None=StaticSwitchParameter(True=NamedRerouteDeclaration(None=@n62) {name=UVVis_1, nodeColor=<rgba 0.0312 1 0 1>}, False=Multiply(A=@n63, B=@n64)) {group=Styx_Transit, parameterName=UsePreskinnedAlignedUV}) {name=UV_withParallax_A_2, nodeColor=<rgba 1 0.6562 0 1>}
  n65 = ScalarParameter() {group=Styx, parameterName=StyxNoise_SpeedX}
  n66 = ScalarParameter() {defaultValue=-0.1, group=Styx, parameterName=StyxNoise_SpeedY}
  n67 = TextureObjectParameter() {
   group=Styx
   parameterName=StyxDistortNormal
   samplerType=SAMPLERTYPE_Normal
   texture=/BRCosmetics/Characters/Player/Male/Medium/Bodies/M_MED_GateHound_Spectral/Textures/T_StyxGateHound_N
  }
  n68 = TextureSample(UVs=Panner(Coordinate=NamedRerouteUsage(), Speed=AppendVector(A=@n65, B=@n66)), Tex=@n67) {
   samplerSource=SSM_Wrap_WorldGroupSettings
   samplerType=SAMPLERTYPE_Normal
   texture=/BRCosmetics/Characters/Player/Male/Medium/Bodies/M_MED_DarkStance_Pater/Textures/T_StyxCosmetic_N
  }
  NamedRerouteDeclaration(None=ComponentMask(None=@n68.RGB) {g=true, r=true}) {name=Distortion_NormalNoise_A_2, nodeColor=<rgba 0.3828 0 1 1>}
  n69 = TextureSampleParameter2D(UVs=@n25) {
   group=StyxMod
   parameterName=DistortSmoke
   samplerSource=SSM_Wrap_WorldGroupSettings
   samplerType=SAMPLERTYPE_Grayscale
   texture=&T_FractalDistortMask_GateHound
  }
  n70 = Saturate(None=Multiply(A=@n69.RGB, B=ScalarParameter() {defaultValue=10, group=StyxMod, parameterName=SmokeAreaMulti}) {constB=5})
  n71 = ScalarParameter() {defaultValue=0.5, group=StyxMod, parameterName=SmokeAreaIntensity}
  NamedRerouteDeclaration(None=LinearInterpolate(A=Multiply(A=@n70, B=@n71), B=NamedRerouteUsage(), Alpha=@n28)) {name=SmokeArea, nodeColor=<rgba 0 1 0.1797 1>}
  n72 = ScalarParameter() {group=Styx_Detail, parameterName=Styx_Detail_Noise_Low}
  n73 = ScalarParameter() {defaultValue=1, group=Styx_Detail, parameterName=Styx_Detail_Noise_High}
  n74 = Panner(
   Coordinate=Add(A=NamedRerouteUsage(), B=NamedRerouteUsage())
   Speed=AppendVector(
    A=ScalarParameter() {group=Styx_Detail, parameterName=Styx_Detail_Noise_SpeedX}
    B=ScalarParameter() {defaultValue=-0.1, group=Styx_Detail, parameterName=Styx_Detail_Noise_SpeedY}
   )
  )
  n75 = TextureSampleParameter2D(UVs=@n74) {
   group=Styx_Detail
   parameterName=StyxNoiseDetail
   samplerSource=SSM_Wrap_WorldGroupSettings
   samplerType=SAMPLERTYPE_Masks
   texture=&T_StyxGateHound_DetailNoise_M
  }
  NamedRerouteDeclaration(None=Saturate(None=LinearInterpolate(A=@n72, B=@n73, Alpha=@n75.R))) {name=DetailNoise_1, nodeColor=<rgba 1 0.0938 0 1>}
  n76 = TextureSampleParameter2D(UVs=@n74) {
   group=StyxMod
   parameterName=DistortSmoke
   samplerSource=SSM_Wrap_WorldGroupSettings
   samplerType=SAMPLERTYPE_Grayscale
   texture=&T_FractalDistortMask_GateHound
  }
  n77 = Multiply(A=@n76.RGB, B=ScalarParameter() {defaultValue=10, group=StyxMod, parameterName=SmokeAreaMulti}) {constB=5}
  NamedRerouteDeclaration(None=Multiply(A=Saturate(None=@n77), B=ScalarParameter() {defaultValue=0.5, group=StyxMod, parameterName=SmokeAreaIntensity})) {name=SmokeArea_1, nodeColor=<rgba 0 1 0.1797 1>}
  NamedRerouteDeclaration(None=Multiply(A=NamedRerouteUsage(), B=ScalarParameter() {defaultValue=0.01, group=Styx, parameterName=Styx_Distort_UV2})) {name=Distortion_Final_A_2, nodeColor=<rgba 0.2422 0 1 1>}
  n78 = NamedRerouteUsage()
  n79 = AppendVector(
   A=ScalarParameter() {group=Styx, parameterName=StyxNoise_Speed2X}
   B=ScalarParameter() {defaultValue=-0.1, group=Styx, parameterName=StyxNoise_Speed2Y}
  )
  n80 = TextureSample(UVs=Add(A=Panner(Coordinate=@n78, Speed=@n79), B=NamedRerouteUsage()), Tex=NamedRerouteUsage()) {samplerSource=SSM_Wrap_WorldGroupSettings}
  n81 = Power(Base=@n80.R, Exp=ScalarParameter() {defaultValue=5, group=Styx, parameterName=StyxNoise_Contrast2})
  n82 = Saturate(None=Power(Base=@n80.R, Exp=ScalarParameter() {defaultValue=5, group=Styx, parameterName=StyxNoise_Contrast}))
  NamedRerouteDeclaration(None=Multiply(A=LinearInterpolate(A=@n81, B=@n80.R, Alpha=@n82), B=ScalarParameter() {defaultValue=1, group=Styx, parameterName=MainNoiseIntensity})) {name=MainNoisePattern_1, nodeColor=<rgba 0 1 0.3203 1>}
  n83 = NamedRerouteUsage()
  n84 = ScalarParameter() {defaultValue=4, group=Styx_Transit, parameterName=StyxEmissiveMulti}
  n85 = Multiply(A=@n84) {constB=0.5}
  n86 = ComponentMask(None=NamedRerouteUsage()) {r=true}
  n87 = Add(A=NamedRerouteUsage(), B=NamedRerouteUsage())
  n88 = NamedRerouteUsage()
  n89 = LinearInterpolate(
   A=LinearInterpolate(B=LinearInterpolate(A=@n84, B=@n85, Alpha=@n86), Alpha=LinearInterpolate(A=@n87, Alpha=@n88) {constB=0}) {constA=1}
   B=Constant() {r=1}
   Alpha=NamedRerouteUsage()
  )
  n90 = NamedRerouteUsage()
  n91 = NamedRerouteUsage()
  n92 = ScalarParameter() {defaultValue=1, group=Styx_Transit, parameterName=FadeMulti}
  n93 = SmoothStep(
   Min=ScalarParameter() {group=Styx_Transit, parameterName=FadeMin}
   Max=ScalarParameter() {defaultValue=0.7, group=Styx_Transit, parameterName=FadeMax}
   Value=Saturate(None=Multiply(A=@n91, B=@n92))
  )
  n94 = Multiply(A=@n93, B=NamedRerouteUsage())
  n95 = Multiply(
   A=Multiply(A=Multiply(A=@n89, B=@n90), B=StaticSwitchParameter(True=@n94, False=@n93) {group=StyxMod, parameterName=UseAreaMask})
   B=NamedRerouteUsage()
  )
  NamedRerouteDeclaration(None=Add(A=Add(A=@n83, B=@n95), B=NamedRerouteUsage())) {name=MainEmissive, nodeColor=<rgba 1 0 0.1875 1>}
  n96 = ScalarParameter() {group=Styx, parameterName=StyxNoise_SpeedX}
  n97 = ScalarParameter() {defaultValue=-0.1, group=Styx, parameterName=StyxNoise_SpeedY}
  n98 = TextureObjectParameter() {
   group=Styx
   parameterName=StyxDistortNormal
   samplerType=SAMPLERTYPE_Normal
   texture=&T_StyxGateHound_N
  }
  n99 = TextureSample(UVs=Panner(Coordinate=NamedRerouteUsage(), Speed=AppendVector(A=@n96, B=@n97)), Tex=@n98) {
   samplerSource=SSM_Wrap_WorldGroupSettings
   samplerType=SAMPLERTYPE_Normal
   texture=&T_StyxCosmetic_N
  }
  NamedRerouteDeclaration(None=ComponentMask(None=@n99.RGB) {g=true, r=true}) {name=Distortion_NormalNoise_A, nodeColor=<rgba 0.3828 0 1 1>}
  NamedRerouteDeclaration(None=FunctionInput() {inputName=CowlMask, sortPriority=11}) {name=CowlMask, nodeColor=<rgba 1 0 0.7734 1>}
  n100 = TextureSampleParameter2D() {
   group=MistRaven
   parameterName=FaceMask
   samplerSource=SSM_Wrap_WorldGroupSettings
   samplerType=SAMPLERTYPE_Masks
   texture=&T_BlackMask
  }
  NamedRerouteDeclaration(None=@n100.RGB) {name=FaceMask, nodeColor=<rgba 0 1 0.8594 1>}
  NamedRerouteDeclaration(None=@n55.RGB) {name=EmissiveMaskTexture, nodeColor=<rgba 0.3359 0 1 1>}
  NamedRerouteDeclaration(None=@n38.RGB) {name=BodyMask, nodeColor=<rgba 0 0.5547 1 1>}
  n101 = VectorParameter() {group=MistRaven, parameterName=FakeSkinTint, defaultValue=<rgba 0.1356 0.0006 0.0242 1>}
  n102 = MaterialFunctionCall(
   Value_Morning=ScalarParameter() {defaultValue=1, group=MistRaven, parameterName=FakeSkinTint_Morning}
   Value_Day=ScalarParameter() {defaultValue=1, group=MistRaven, parameterName=FakeSkinTint_day}
   Value_Evening=ScalarParameter() {defaultValue=1, group=MistRaven, parameterName=FakeSkinTint_Evening}
   Value_Night=ScalarParameter() {defaultValue=1, group=MistRaven, parameterName=FakeSkinTint_Night}
  ) {materialFunction=&Scalar_Time_of_Day}
  NamedRerouteDeclaration(None=NamedRerouteDeclaration(None=Multiply(A=@n101.RGB, B=@n102.Result)) {name=SkinTint, nodeColor=<rgba 0.8047 0 1 1>}) {name=SSS_ColorOUT, nodeColor=<rgba 0.0547 1 0 1>}
  NamedRerouteDeclaration(None=@n9.MaterialAttributes) {name=MA, nodeColor=<rgba 0 1 0.8125 1>}
  NamedRerouteDeclaration(None=Multiply(A=NamedRerouteUsage(), B=ScalarParameter() {defaultValue=0.01, group=Styx, parameterName=Styx_Distort_UV2})) {name=Distortion_Final_A, nodeColor=<rgba 0.2422 0 1 1>}
  n103 = Constant() {r=1}
  n104 = TextureCoordinate()
  n105 = Multiply(A=Subtract(A=ComponentMask(None=@n104) {g=true}) {constB=0.75}) {constB=5}
  n106 = LinearInterpolate(Alpha=Multiply(A=Saturate(None=@n105), B=@n100.R)) {constA=0.15, constB=0}
  n107 = ComponentMask(None=@n34) {b=true}
  n108 = SmoothStep(Max=ScalarParameter() {defaultValue=0.24, group=MistRaven, parameterName=NeckBrightness}, Value=Saturate(None=Divide(A=@n107) {constB=10})) {constMax=0.4}
  NamedRerouteDeclaration(None=StaticSwitchParameter(True=Multiply(A=@n103, B=@n106) {constB=0.2}, False=OneMinus(None=@n108)) {group=Styx_Transit, parameterName=IsNeck}) {name=NeckMask, nodeColor=<rgba 1 0 0.2344 1>}
  n109 = TextureObjectParameter() {
   group=Styx
   parameterName=Styx Mask
   samplerType=SAMPLERTYPE_Masks
   texture=&T_StyxGateHound_Mask
  }
  NamedRerouteDeclaration(None=@n109) {name=Styx Mask IN, nodeColor=<rgba 1 0.8906 0 1>}
  TextureCoordinate()
 ]}
