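# Parent skeletal-mesh material for the MistRaven Cowl character body skin, layering a cowl emissive and transition effect onto the shared 2023 character material.
# The PreFX output feeds the Cowl emissive function, whose output drives the transition function alongside pre-skinned position.
# The transition result blends over the base attributes using the boosted Tech Art mask blue channel, with IsHead lerping that alpha toward the default constant.
# PostFX finishes the blended attributes.
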Material /BRCosmetics/Characters/Player/Female/Medium/Bodies/F_MED_MistRaven_Peck/Skins/Cowl/Materials/M_Parent_MistRaven_Cowl_Body
graph {
 properties {
  bUseMaterialAttributes=true
  bUsedWithMeshDeformer=true
  bUsedWithMorphTargets=true
  bUsedWithSkeletalMesh=true
  pixelDepthOffsetMode=PDOM_Legacy
  shadingModel=MSM_FromMaterialExpression
  usedShadingModels=Default Lit | Subsurface | Clear Coat | Subsurface Profile | Cloth
 }
 nodes [
  n1 = MaterialFunctionCall!(Material=NamedRerouteUsage()) {materialFunction=&/MF_MistRaven_Cowl_Body_Emissive}
  n2 = MaterialFunctionCall() {materialFunction=&MF_CharacterParent_2023_PreFX}
  n3 = MaterialFunctionCall!(
   EmissiveTextureSampler=@n1.Emissive Mask
   Material=@n1.Emissive Mask
   PreSkinnedLocation=@n2.Pre skinned Local Position
   CowlMask=@n1.Emissive Mask
  ) {materialFunction=&/MF_Misraven_Cowl_Transition}
  n4 = ComponentMask(None=Saturate(None=Multiply(A=@n2.Tech Art Mask) {constB=3})) {b=true}
  n5 = ScalarParameter() {parameterName=IsHead}
  MaterialFunctionCall(
   Material=BlendMaterialAttributes(A=NamedRerouteUsage(), B=@n3.Result, Alpha=LinearInterpolate(A=@n4, Alpha=@n5))
   M=@n2.M
   Pre Skinned Local Position=@n2.M
   Distance Blend=@n2.M
   Fresnel=@n2.M
  ) {materialFunction=&MF_CharacterParent_2023_PostFX}
  NamedRerouteDeclaration(None=@n2.Result) {name=Input MA, nodeColor=<rgba 1 0 0.1641 1>}
 ]
}
